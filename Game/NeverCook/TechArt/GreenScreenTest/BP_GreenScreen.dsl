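; BP_GreenScreen is a tech-art test Blueprint that renders green screen mask textures by drawing materials into render targets.
; Every tick it runs three passes in order: a mask bake, then two blur materials, the second named as an inside blur, each drawn to its own render target.
; Custom event TestRender repeats all three passes, while Blur redraws only the two blur targets.
; BeginPlay, overlap and the construction script are empty.
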(event EventBeginPlay)

(event Collision|EventActorBeginOverlap
  (params (OtherActor :type Actor%20Object%20Reference)))

(event Custom|TestRender
  (Rendering|DrawMaterialtoRenderTarget "/Game/NeverCook/TechArt/GreenScreenTest/GreenScreen_Mask_RT.GreenScreen_Mask_RT" "/Game/NeverCook/TechArt/GreenScreenTest/M_GreenScreen_BakeMasktoRT.M_GreenScreen_BakeMasktoRT")
  (Rendering|DrawMaterialtoRenderTarget "/Game/NeverCook/TechArt/GreenScreenTest/GreenScreen_MaskBlur_RT.GreenScreen_MaskBlur_RT" "/Game/NeverCook/TechArt/GreenScreenTest/M_GreenscreenBlur_Inst.M_GreenscreenBlur_Inst")
  (Rendering|DrawMaterialtoRenderTarget "/Game/NeverCook/TechArt/GreenScreenTest/GreenScreen_MaskBlur_RT1.GreenScreen_MaskBlur_RT1" "/Game/NeverCook/TechArt/GreenScreenTest/M_GreenscreenBlur_Inside.M_GreenscreenBlur_Inside"))

(event EventTick
  (params (DeltaSeconds :type Float%20%28single-precision%29))
  (Rendering|DrawMaterialtoRenderTarget "/Game/NeverCook/TechArt/GreenScreenTest/GreenScreen_Mask_RT.GreenScreen_Mask_RT" "/Game/NeverCook/TechArt/GreenScreenTest/M_GreenScreen_BakeMasktoRT.M_GreenScreen_BakeMasktoRT")
  (Rendering|DrawMaterialtoRenderTarget "/Game/NeverCook/TechArt/GreenScreenTest/GreenScreen_MaskBlur_RT.GreenScreen_MaskBlur_RT" "/Game/NeverCook/TechArt/GreenScreenTest/M_GreenscreenBlur_Inst.M_GreenscreenBlur_Inst")
  (Rendering|DrawMaterialtoRenderTarget "/Game/NeverCook/TechArt/GreenScreenTest/GreenScreen_MaskBlur_RT1.GreenScreen_MaskBlur_RT1" "/Game/NeverCook/TechArt/GreenScreenTest/M_GreenscreenBlur_Inside.M_GreenscreenBlur_Inside"))

(event Custom|Blur
  (Rendering|DrawMaterialtoRenderTarget "/Game/NeverCook/TechArt/GreenScreenTest/GreenScreen_MaskBlur_RT.GreenScreen_MaskBlur_RT" "/Game/NeverCook/TechArt/GreenScreenTest/M_GreenscreenBlur_Inst.M_GreenscreenBlur_Inst")
  (Rendering|DrawMaterialtoRenderTarget "/Game/NeverCook/TechArt/GreenScreenTest/GreenScreen_MaskBlur_RT1.GreenScreen_MaskBlur_RT1" "/Game/NeverCook/TechArt/GreenScreenTest/M_GreenscreenBlur_Inside.M_GreenscreenBlur_Inside"))

(fn ConstructionScript)
