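# Translucent unlit tech-art test material that chroma-keys a green-screen image and surrounds the subject with an animated glow.
# Chroma key alpha and despill produce the cutout, which blends over two-color glow built from blurred mask render targets.
# Radial sine waves distort the glow into moving spikes.
# The IsMaskBakingMaterial switch instead outputs the scaled mask as emissive, apparently to bake those targets; fake-depth displacement and billboard nodes are left unconnected.
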Material /Game/NeverCook/TechArt/GreenScreenTest/M_GreenScreen
graph {
 properties {
  bUseMaterialAttributes=true
  blendMode=BLEND_Translucent
  pixelDepthOffsetMode=PDOM_Legacy
  shadingModel=MSM_Unlit
  twoSided=true
  usedShadingModels=Unlit
 }
 nodes [
  n1 = TextureCoordinate()
  n2 = MaterialFunctionCall(UVs=@n1, Texture Scale=Constant2Vector() {g=0.6, r=0.75}) {materialFunction=&ScaleUVsByCenter}
  n3 = StaticSwitchParameter(True=@n2.UVs, False=@n1) {parameterName=IsMaskBakingMaterial}
  n4 = TextureSample(UVs=@n3) {samplerType=SAMPLERTYPE_Grayscale, texture=&/GreenScreenTest_FakeDepth}
  n5 = VertexNormalWS()
  n6 = ScalarParameter() {parameterName=Displace}
  n7 = ScalarParameter() {parameterName=Displace1}
  n8 = MaterialFunctionCall(Float2=@n3) {materialFunction=&BreakOutFloat2Components}
  n9 = Subtract(A=CameraPositionWS(), B=WorldPosition().XYZ)
  n10 = Constant3Vector() {constant=<rgba 0 -1 0 0>}
  n11 = Normalize(VectorInput=CrossProduct(A=@n9, B=Transform(None=@n10)))
  n12 = TextureCoordinate()
  n13 = Multiply(A=Multiply(A=Add(A=@n12) {constB=-0.5}) {constB=2}, B=ScalarParameter() {defaultValue=19.8524, parameterName=size})
  n14 = ComponentMask(None=@n13) {r=true}
  n15 = ComponentMask(None=@n13) {g=true}
  n16 = Constant3Vector() {constant=<rgba 1 0 0 0>}
  n17 = Normalize(VectorInput=CrossProduct(A=@n9, B=Transform(None=@n16)))
  Add(
   A=Multiply(A=Multiply(A=@n4.R, B=@n5), B=LinearInterpolate(A=@n6, B=@n7, Alpha=@n8.G))
   B=Add(A=Multiply(A=@n11, B=@n14), B=Multiply(A=@n15, B=@n17))
  )
  Constant() {r=1}
  Constant() {r=1}
  MaterialFunctionCall!(Center Location=ObjectPositionWS(), 0-1 UVs, XY scale, OverridePivot, "Normals (Optional)") {materialFunction=&Sprite}
  n18 = MaterialFunctionCall(UVs=TextureCoordinate(), Texture Scale=Divide(A=Constant2Vector() {g=1, r=1}, B=Constant2Vector() {g=0.6, r=0.75})) {materialFunction=&ScaleUVsByCenter}
  n19 = TextureSample(UVs=@n18.UVs) {samplerType=SAMPLERTYPE_LinearColor, texture=&/GreenScreen_MaskBlur_RT1}
  OneMinus(None=@n19.R)
  n20 = TextureSample(UVs=@n3) {texture=&/GreenscreentestImage}
  n21 = MaterialFunctionCall!(
   LumaMask
   Image Color=@n20.RGB
   Chroma Color=VectorParameter() {parameterName=Chroma, defaultValue=<rgba 0.0762 0.6105 0.2159 1>}.RGB
   Alpha Cutoff Min=ScalarParameter() {parameterName=Min}
   Alpha Cutoff Max=ScalarParameter() {defaultValue=1, parameterName=Max}
   Alpha Exponent
   Despill Cuttoff Max=ScalarParameter() {parameterName=MaxDespill}
   Despill Exponent
  ) {materialFunction=/Engine/Functions/Engine_MaterialFunctions01/Opacity/Chroma_Key_Alpha}
  n22 = ScalarParameter() {defaultValue=1, parameterName=AlphaBoost}
  n23 = Multiply(
   A=MaterialFunctionCall(UVs=@n3) {materialFunction=&ScaleUVsByCenter}.0-1 mask
   B=Clamp(None=Multiply(A=@n21.Alpha, B=@n22))
  )
  n24 = MakeMaterialAttributes!(
   BaseColor
   Metallic
   Specular
   Roughness
   Anisotropy
   EmissiveColor=MaterialFunctionCall!(Image Color=@n20.RGB) {materialFunction=/Engine/Functions/Engine_MaterialFunctions01/Chromakeying/DespillByAvg}.Result
   Opacity=@n23
   OpacityMask
   Normal
   Tangent
   WorldPositionOffset
   SubsurfaceColor
   ClearCoat
   ClearCoatRoughness
   AmbientOcclusion
   Refraction
   CustomizedUVs_0
   CustomizedUVs_1
   CustomizedUVs_2
   CustomizedUVs_3
   CustomizedUVs_4
   CustomizedUVs_5
   CustomizedUVs_6
   CustomizedUVs_7
   PixelDepthOffset
   ShadingModel
   Displacement
  )
  n25 = GetMaterialAttributes(None=@n24) {attributeGetTypes [#1]}
  Saturate(None=LinearInterpolate(
    A=ScalarParameter() {defaultValue=-0.2857, parameterName=SoftEdgeBlend_1}
    B=ScalarParameter() {defaultValue=1.4398, parameterName=SoftEdgeBlend_2}
    Alpha=@n25.Opacity
   ))
  n26 = Subtract(A=TextureCoordinate(), B=Constant2Vector() {g=0.5, r=0.5})
  n27 = Normalize(VectorInput=@n26)
  n28 = DotProduct(A=@n27, B=Constant2Vector() {g=-1})
  n29 = Power(Base=Divide(A=Add(A=@n28)), Exp=ScalarParameter() {defaultValue=0.8978, parameterName=StarDistortFalloffBottom})
  n30 = ComponentMask(None=MaterialFunctionCall(Vector or UVs=@n26) {materialFunction=&VectorToRadialValue}.Radial Coordinates) {r=true}
  n31 = Multiply(A=@n30, B=ScalarParameter() {defaultValue=10.8343, parameterName=StarDistort_Frequency})
  n32 = Multiply(A=Time(), B=ScalarParameter() {defaultValue=0.5, parameterName=StarDistort_Speed})
  n33 = Power(Base=Sine(None=Add(A=@n31, B=@n32)), Exp=ScalarParameter() {defaultValue=11.749, parameterName=StarDistort_Shape})
  n34 = Multiply(A=@n30, B=ScalarParameter() {defaultValue=9.206, parameterName=StarDistort_Frequency2})
  n35 = Multiply(A=Time(), B=ScalarParameter() {defaultValue=-0.45, parameterName=StarDistort_Speed2})
  n36 = Power(Base=Sine(None=Add(A=@n34, B=@n35)), Exp=ScalarParameter() {defaultValue=11.749, parameterName=StarDistort_Shape2})
  n37 = Multiply(A=Multiply(A=Saturate(None=@n29), B=Add(A=@n33, B=@n36)), B=ScalarParameter() {defaultValue=-0.0952, parameterName=StarDistort_Length})
  n38 = TextureSample(UVs=Add(A=@n18.UVs, B=Multiply(A=@n27, B=@n37))) {samplerType=SAMPLERTYPE_LinearColor, texture=&/GreenScreen_MaskBlur_RT}
  n39 = Power(Base=@n38.R, Exp=ScalarParameter() {defaultValue=3.3471, parameterName=OutlinePower_2})
  n40 = LinearInterpolate(
   A=VectorParameter() {parameterName=ColorGlow, defaultValue=<rgba 3 0 2.0086 1>}.RGB
   B=VectorParameter() {parameterName=ColorGlow2, defaultValue=<rgba 20 0 14.5937 1>}.RGB
   Alpha=@n39
  )
  n41 = ScalarParameter() {defaultValue=1.044, parameterName=OutlinePower}
  n42 = Saturate(None=Add(A=Power(Base=@n38.R, Exp=@n41), B=@n39))
  n43 = Power(Base=@n19.R, Exp=ScalarParameter() {defaultValue=1.917, parameterName=InnerGlowTransitionPower})
  n44 = ScalarParameter() {defaultValue=2.1692, parameterName=InnerGlowTransitionMult}
  n45 = Multiply(A=@n25.Opacity, B=OneMinus(None=Multiply(A=@n43, B=@n44)))
  StaticSwitchParameter(
   True=SetMaterialAttributes!(MaterialAttributes, Emissive Color=Multiply(A=@n2.0-1 mask, B=@n23), Opacity=Constant() {r=1}) {attributeSetTypes [#2, #1]}
   False=BlendMaterialAttributes(
    A=SetMaterialAttributes(Emissive Color=@n40, Opacity=@n42) {attributeSetTypes [#2, #1]}
    B=@n24
    Alpha=Saturate(None=@n45)
   )
  ) {parameterName=IsMaskBakingMaterial}
 ]
}
