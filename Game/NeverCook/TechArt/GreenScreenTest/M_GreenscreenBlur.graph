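# M_GreenscreenBlur is an unlit translucent test material that softens a green screen mask render target.
# The engine's SpiralBlur-Texture function samples the mask with clamped UVs, with parameters for distance, step counts, radial offset and kernel power.
# The blurred red channel passes through an Invert static switch that can flip the mask.
# A center-based UV scale node is present but does not feed the blur.
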Material /Game/NeverCook/TechArt/GreenScreenTest/M_GreenscreenBlur
graph {
 properties {blendMode=BLEND_Translucent, pixelDepthOffsetMode=PDOM_Legacy, shadingModel=MSM_Unlit, usedShadingModels=Unlit}
 nodes [
  Constant() {r=1}
  n1 = TextureCoordinate()
  MaterialFunctionCall(UVs=@n1, Texture Scale=Constant2Vector() {g=0.6, r=0.75}) {materialFunction=&ScaleUVsByCenter}
  n2 = ComponentMask(None=MaterialFunctionCall!(
    TextureObject=TextureObjectParameter() {parameterName=Param, samplerType=SAMPLERTYPE_LinearColor, texture=&/GreenScreen_Mask_RT}
    UV=Clamp(None=@n1)
    Distance=ScalarParameter() {defaultValue=0.05, parameterName=Distance}
    Distance Steps=ScalarParameter() {defaultValue=16, parameterName=StepDistance}
    Radial Steps=ScalarParameter() {defaultValue=8, parameterName=RadialStep}
    Radial Offset=ScalarParameter() {defaultValue=0.62, parameterName=Offset}
    Kernel Power=ScalarParameter() {defaultValue=1, parameterName=Power}
   ) {materialFunction=/Engine/Functions/Engine_MaterialFunctions01/Texturing/SpiralBlur-Texture}.Result) {r=true}
  StaticSwitchParameter(True=OneMinus(None=@n2), False=@n2) {parameterName=Invert}
 ]
}
