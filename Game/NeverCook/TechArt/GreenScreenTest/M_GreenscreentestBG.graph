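# M_GreenscreentestBG is an unlit material for a green screen test background, kept under TechArt.
# Its graph holds a single vector parameter, GreenScreencolor, which defaults to chroma green and lets material instances override the color.
# The graph shows no link from that parameter to a material output such as Emissive Color.
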Material /Game/NeverCook/TechArt/GreenScreenTest/M_GreenscreentestBG
graph {
 properties {pixelDepthOffsetMode=PDOM_Legacy, shadingModel=MSM_Unlit, usedShadingModels=Unlit}
 nodes [VectorParameter() {parameterName=GreenScreencolor, defaultValue=<rgba 0 0.4397 0.0513 1>}]
}
