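# M_TESTMASK is an unlit test material for a green-screen mask, kept in a NeverCook TechArt test folder.
# It scales UVs about their center with ScaleUVsByCenter, using the reciprocal of a constant non-uniform scale.
# Those UVs sample the red channel of the GreenScreen_MaskBlur_RT texture, which is multiplied by the function's 0-1 mask.
# The graph shows no connection from that product to a material output.
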Material /Game/NeverCook/TechArt/GreenScreenTest/M_TESTMASK
graph {
 properties {pixelDepthOffsetMode=PDOM_Legacy, shadingModel=MSM_Unlit, usedShadingModels=Unlit}
 nodes [
  n1 = MaterialFunctionCall(UVs=TextureCoordinate(), Texture Scale=Divide(A=Constant2Vector() {g=1, r=1}, B=Constant2Vector() {g=0.6, r=0.75})) {materialFunction=&ScaleUVsByCenter}
  Multiply(
   A=TextureSample(UVs=@n1.UVs) {samplerType=SAMPLERTYPE_LinearColor, texture=&/GreenScreen_MaskBlur_RT}.R
   B=@n1.0-1 mask
  )
 ]
}
